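# Physics asset for the Snow Ninja female medium backpack accessory mesh, giving it a simple collision shape.
# A single box body on the backpack_root bone, offset slightly along Y, covers the whole asset.
# It defines no constraints, so the backpack has no jointed or simulated secondary parts.
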PhysicsAsset /Game/Accessories/FORT_Backpacks/F_MED_SnowNinja_Backpack/Meshes/Backpack_F_MED_Snow_Ninja_Physics
bodies {backpack_root [{shapeType=Box, extentX=7.5, extentY=7.5, extentZ=7.5, center=<0 6.168 0>}]}
constraints []
